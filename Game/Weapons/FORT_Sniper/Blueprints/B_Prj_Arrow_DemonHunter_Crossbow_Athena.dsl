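; Demon Hunter crossbow bolt projectile for Athena, handling impact effects when it touches something.
; Overlaps with water surfaces play a splash sound and spawn FX fed with water depth and body index.
; Blocking hits stop the trail, gravity and whip tracker, then cosmetically attach bone-snapped effects to fiends or hostile pawns.
; Hostile pawn hits also play a player or surface-based break/stuck sound, then call DoExplosion after a short delay.
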(event EventOnTouched
  (params (OtherActor :type Actor%20Object%20Reference) (OtherComp :type Primitive%20Component%20Object%20Reference) (HitResult :type Hit%20Result%20Structure%20%28by%20ref%29) (bIsOverlap :type Boolean))
  (bind _hitresult (Variables|Default|GetHitResult))
  (bind (_bblockinghit _binitialoverlap _time _distance _location _impactpoint _normal _impactnormal _physmat) (Collision|BreakHitResult _hitresult))
  (bind _waterinteractioncomponent (Variables|Water|GetWaterInteractionComponent))
  (bind (_waterplanelocation _waterplanenormal _watersurfaceposition _waterdepth _waterbodyidx) (FX|GetLastWaterSurfaceInfo _waterinteractioncomponent :!class /Script/FortniteGame.FortWaterInteractionComponent))
  (bind _hitresult_2 (Variables|Default|GetHitResult))
  (bind (_bblockinghit_2 _binitialoverlap_2 _time_2 _distance_2 _location_2 _impactpoint_2 _normal_2 _impactnormal_2 _physmat_2 _hitactor) (Collision|BreakHitResult _hitresult_2))
  (bind (_asfort_player_pawn _bsuccess) (Utilities|Casting|CastToFortPlayerPawn _hitactor))
  (if bIsOverlap
    (Variables|Default|SetHitResult HitResult)
    ;[0 0.298 1] Water Impact Projectile
    (Networking|CanExecuteCosmeticEvents
      (:True
        (Utilities|IsValid _physmat
          (:Is%20Valid
            (if (Utilities|Enum|Equal(Enum) (Variables|/Script/PhysicsCore.PhysicalMaterial|GetSurfaceType _physmat) "SurfaceType8")
              (Audio|PlaySoundatLocation "/Game/Athena/Sounds/Impacts/Sniper_Rifle_Impact_Water_Cue.Sniper_Rifle_Impact_Water_Cue" _location)
              (bind _returnvalue (Effects|SpawnEmitteratLocation (Variables|Default|GetWaterImpactFX) _location (Math|Conversions|RotationFromXVector _waterplanenormal)))
              (Effects|Components|ParticleSystem|SetFloatParameter _returnvalue "WaterDepth" _waterdepth)
              (Effects|Components|ParticleSystem|SetFloatParameter _returnvalue "WaterBodyIndex" (Math|Conversions|ToFloat(Integer) _waterbodyidx)))))))
    (else
      (Components|Activation|SetActive (Variables|Default|GetP_Crossbow_Projectile) false false)
      (Variables|/Script/Engine.ProjectileMovementComponent|SetProjectileGravityScale 0.0 (Variables|Movement|GetProjectileMovementComponent))
      (Class|BulletWhipTrackerComponentBase|Disable (Variables|Default|GetBulletWhipTrackerComponent_Crossbow))
      (Networking|CanExecuteCosmeticEvents
        (:True
          (bind _asdeimos_fiend (Utilities|Casting|CastToDeimos_Fiend _hitactor)
            (:then
              ;[1 1 1] Deimos impact VFX
              (bind _returnvalue_2 (Effects|SpawnEmitterAttached :EmitterTemplate (Variables|Default|GetHitDeimosEffect) :AttachToComponent (Variables|/Script/Engine.Character|GetMesh _asdeimos_fiend) :AttachPointName (Class|DeimosFiend|TempFindClosestBone _asdeimos_fiend (Transformation|GetActorLocation)) :Scale "1.000000,1.000000,1.000000" :LocationType "SnapToTarget" :bAutoDestroy true :PoolingMethod "None" :bAutoActivate true))
              (Transformation|SetWorldRotation _returnvalue_2 (Math|Conversions|RotationFromXVector (* (Transformation|GetActorForwardVector) -1.0)) false false))
            (:CastFailed
              (bind _asfort_pawn (Utilities|Casting|CastToFortPawn _hitactor)
                (:then
                  ;[1 1 1] Hostile player impact VFX
                  (switch Utilities|FlowControl|Switch|SwitchonEFortTeamAffiliation (Team|GetTeamAffiliation _asfort_pawn self)
                    (:Friendly
                      _)
                    (:Neutral
                      _)
                    (:Hostile
                      (bind _returnvalue_3 (Effects|SpawnEmitterAttached :EmitterTemplate (Variables|Default|GetHitPlayerEffect) :AttachToComponent (Variables|/Script/Engine.Character|GetMesh _asfort_pawn) :AttachPointName (select (Utilities|Name|Equal(Name) (.ReturnValue (Components|SkinnedMesh|FindClosestBone (Variables|/Script/Engine.Character|GetMesh _asfort_pawn) (Transformation|GetActorLocation) 0.0 false)) "head") "neck_01" (.ReturnValue (Components|SkinnedMesh|FindClosestBone (Variables|/Script/Engine.Character|GetMesh _asfort_pawn) (Transformation|GetActorLocation) 0.0 false))) :Scale "1.000000,1.000000,1.000000" :LocationType "SnapToTarget" :bAutoDestroy true :PoolingMethod "None" :bAutoActivate true))
                      (Transformation|SetWorldRotation _returnvalue_3 (Math|Conversions|RotationFromXVector (* (Transformation|GetActorForwardVector) -1.0)) false false)
                      (if (> (Utilities|Select :Index (Collision|GetSurfaceType (Variables|Default|GetHitResult)) :SurfaceType_Default 1 :SurfaceType1 0 :SurfaceType2 1 :SurfaceType3 1 :SurfaceType4 0 :SurfaceType5 0 :SurfaceType6 0 :SurfaceType7 0 :SurfaceType8 0 :SurfaceType9 0 :SurfaceType10 0 :SurfaceType11 0 :SurfaceType12 0 :SurfaceType13 0 :SurfaceType14 0 :SurfaceType15 0 :SurfaceType16 0 :SurfaceType17 0 :SurfaceType18 0 :SurfaceType19 0 :SurfaceType20 0 :SurfaceType21 0 :SurfaceType22 0 :SurfaceType23 0 :SurfaceType24 0 :SurfaceType25 0 :SurfaceType26 0 :SurfaceType27 0 :SurfaceType28 0 :SurfaceType29 0 :SurfaceType30 0 :SurfaceType31 0 :SurfaceType32 0 :SurfaceType33 0 :SurfaceType34 0) 0)
                        (Variables|Default|SetBolt%20Surface%20Impact (Variables|Default|GetBoltBreak))
                        (else
                          (Variables|Default|SetBolt%20Surface%20Impact (Variables|Default|GetBoltStuck))))
                      (Audio|PlaySoundatLocation (select _bsuccess "/Game/Athena/Sounds/Impacts/Crossbow_DemonHunter_Player_Impact_Athena_Cue.Crossbow_DemonHunter_Player_Impact_Athena_Cue" (Variables|Default|GetBolt%20Surface%20Impact)) (Transformation|GetActorLocation))
                      (Utilities|FlowControl|Delay 0.5)
                      (Delivery|DoExplosion))))
                (:CastFailed
                  _)))))
        (:False
          _)))))

(fn ConstructionScript)
