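; Valentine crossbow arrow projectile for Athena, handling water, impact and hit effects.
; On touch it stores the hit: water overlaps spawn a depth-aware splash and destroy the arrow, while other hits stop trail, gravity and bullet whip.
; The explosion fires only where cosmetic events cannot run.
; On explode, cosmetic instances play impact audio and attach a Niagara effect to hostile pawns' closest bone, substituting neck_01 for head.
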(event EventOnTouched
  (params (OtherActor :type Actor%20Object%20Reference) (OtherComp :type Primitive%20Component%20Object%20Reference) (HitResult :type Hit%20Result%20Structure%20%28by%20ref%29) (bIsOverlap :type Boolean))
  (bind (_bblockinghit _binitialoverlap _time _distance _location _impactpoint _normal _impactnormal _physmat) (Collision|BreakHitResult HitResult))
  (bind _waterinteractioncomponent (Variables|Water|GetWaterInteractionComponent))
  (bind (_waterplanelocation _waterplanenormal _watersurfaceposition _waterdepth _waterbodyidx) (FX|GetLastWaterSurfaceInfo _waterinteractioncomponent :!class /Script/FortniteGame.FortWaterInteractionComponent))
  (bind _output_get (Variables|Default|SetHit%20Result HitResult))
  (if bIsOverlap
    ; WaterImpact
    (Networking|CanExecuteCosmeticEvents
      (:True
        (if (Utilities|Enum|Equal(Enum) (Variables|/Script/PhysicsCore.PhysicalMaterial|GetSurfaceType _physmat) "SurfaceType8")
          (Audio|PlaySoundatLocation "/Game/Athena/Sounds/Impacts/Sniper_Rifle_Impact_Water_Cue.Sniper_Rifle_Impact_Water_Cue" _location)
          (bind _returnvalue (Effects|SpawnEmitteratLocation (Variables|Default|GetWaterImpactFX) _location (Math|Conversions|RotationFromXVector _waterplanenormal)))
          (Effects|Components|ParticleSystem|SetFloatParameter _returnvalue "WaterDepth" _waterdepth)
          (Effects|Components|ParticleSystem|SetFloatParameter _returnvalue "WaterBodyIndex" (Math|Conversions|ToFloat(Integer) _waterbodyidx))
          (Actor|DestroyActor))))
    (else
      (Components|Activation|SetActive (Variables|Default|GetNS_ValentineCrossBow_ProjTrail) false false)
      (Variables|/Script/Engine.ProjectileMovementComponent|SetProjectileGravityScale 0.0 (Variables|Movement|GetProjectileMovementComponent))
      (Class|BulletWhipTrackerComponentBase|Disable (Variables|Visual|GetBulletWhipTrackerComponent))
      (CallFunction|HitBlockingMeshonPlayer? (.HitActor (Collision|BreakHitResult _output_get)))
      (Networking|CanExecuteCosmeticEvents
        (:False
          (if (not (Delivery|IsBeingKilled))
            (Delivery|DoExplosion)))))))

(event EventOnExploded
  (params (HitActors :type Array%20of%20Actor%20Object%20References) (HitResults :type Array%20of%20Hit%20Result%20Structures))
  (bind _hit_result (Variables|Default|GetHit%20Result))
  (bind (_bblockinghit _binitialoverlap _time _distance _location _impactpoint _normal _impactnormal _physmat _hitactor) (Collision|BreakHitResult _hit_result))
  (Networking|CanExecuteCosmeticEvents
    (:True
      (Utilities|Casting|CastToFortPlayerPawn _hitactor)
      (Audio|PlaySoundatLocation (select false "/Game/Athena/Sounds/Impacts/Crossbow_Player_Impact_Athena_Cue.Crossbow_Player_Impact_Athena_Cue" "/Game/Athena/Sounds/Impacts/Crossbow_Surface_Impact_Athena_Cue.Crossbow_Surface_Impact_Athena_Cue") _impactpoint)
      (bind _asfort_pawn (Utilities|Casting|CastToFortPawn _hitactor))
      (switch Utilities|FlowControl|Switch|SwitchonEFortTeamAffiliation (Team|GetTeamAffiliation _asfort_pawn self)
        (:Hostile
          (Niagara|SpawnSystemAttached (Variables|Default|GetNiagaraHitPlayerEffect) (Variables|/Script/Engine.Character|GetMesh _asfort_pawn) (select (Utilities|Name|Equal(Name) (.ReturnValue (Components|SkinnedMesh|FindClosestBone (Variables|/Script/Engine.Character|GetMesh _asfort_pawn) (Transformation|GetActorLocation) 0.0 false)) "head") "neck_01" (.ReturnValue (Components|SkinnedMesh|FindClosestBone (Variables|/Script/Engine.Character|GetMesh _asfort_pawn) (Transformation|GetActorLocation) 0.0 false)))))))))

(fn ConstructionScript
  (|CallParentFunction))
