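; Cowboy sniper bullet projectile that sets scalable gravity on construction and handles trail and impact cosmetics.
; BeginPlay feeds muzzle origin and timing to the Niagara trail, restarting it on mobile as a hacky fix.
; Stopping disables gravity and bullet whip, plays impact audio, retires the trail and decals non-pawn hits off mobile.
; Overlap and trace water hits spawn depth-driven splashes and flag water audio to suppress later impact sounds.
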(event EventBeginPlay
  (bind _ns_coresniper_bullet_trail (Variables|Default|GetNS_CoreSniper_Bullet_Trail))
  (bind _ns_coresniper_bullet_trail_2 (Variables|Default|GetNS_CoreSniper_Bullet_Trail))
  (|CallParentFunction)
  (Networking|CanExecuteCosmeticEvents
    (:True
      (bind _asfort_player_pawn (Utilities|Casting|CastToFortPlayerPawn (Variables|Actor|GetInstigator)))
      ;[1 1 1] Hacky fix for FORT-587865
      (if (Fortnite|IsMobilePlatform)
        (Components|Activation|Deactivate _ns_coresniper_bullet_trail_2)
        (Utilities|FlowControl|Delay 0.04)
        (Components|Activation|Activate _ns_coresniper_bullet_trail_2 false)
        (else
          _))
      (Niagara|SetNiagaraVariable(Vector3) _ns_coresniper_bullet_trail "BulletOrigin" (Effects|GetMuzzleLocation (Variables|/Script/FortniteGame.FortPawn|GetCurrentWeapon _asfort_player_pawn) -1))
      (Niagara|SetNiagaraVariable(Float) _ns_coresniper_bullet_trail "StartTime" (Utilities|Time|GetGameTimeinSeconds))
      (Niagara|SetNiagaraVariable(Float) _ns_coresniper_bullet_trail "WPODuration" 0.4))
    (:False
      (Variables|/Script/Engine.PrimitiveComponent|SetbGenerateOverlapEvents false (Variables|Collision|GetCapsuleComponent)) ; Overlaps only needed on clients for some FX interaction
    )))

(event EventOnStop
  (params (Hit :type Hit%20Result%20Structure%20%28by%20ref%29))
  (bind (_bblockinghit _binitialoverlap _time _distance _location _impactpoint _normal _impactnormal _physmat _hitactor) (Collision|BreakHitResult Hit))
  (bind (_asfort_pawn _bsuccess) (Utilities|Casting|CastToFortPawn _hitactor))
  (Utilities|FlowControl|Sequence
    (:then_0
      (Variables|/Script/Engine.ProjectileMovementComponent|SetProjectileGravityScale 0.0 (Variables|Movement|GetProjectileMovementComponent))
      (Class|BulletWhipTrackerComponentBase|Disable (Variables|Visual|GetBulletWhipTrackerComponent)))
    (:then_1
      (Networking|CanExecuteCosmeticEvents
        (:True
          (CallFunction|PlayImpactSFX _bsuccess (Utilities|Enum|Equal(Enum) (Variables|/Script/PhysicsCore.PhysicalMaterial|GetSurfaceType _physmat) "SurfaceType8"))
          (Transformation|DetachFromComponent (Variables|Default|GetNS_CoreSniper_Bullet_Trail) "KeepWorld" "KeepWorld" "KeepWorld" true)
          (Utilities|FlowControl|Sequence
            (:then_0
              (Niagara|SetNiagaraVariable(Bool) (Variables|Default|GetNS_CoreSniper_Bullet_Trail) "Deactive" true)
              (Utilities|FlowControl|Delay 0.2)
              (Components|Activation|Deactivate (Variables|Default|GetNS_CoreSniper_Bullet_Trail)))
            (:then_1
              (if (not (or (Fortnite|IsMobilePlatform) _bsuccess))
                (bind _returnvalue (Rendering|Decal|SpawnDecalatLocation (Variables|Default|GetImpactDecal%20Material) (Variables|Default|GetDecalSize) _location (Math|Conversions|RotationFromXVector (* _impactnormal -1.0)) 4.5))
                (Rendering|Components|Decal|SetFadeOut _returnvalue 0.0 4.5 true)))))))))

(event EventOnTouched
  (params (OtherActor :type Actor%20Object%20Reference) (OtherComp :type Primitive%20Component%20Object%20Reference) (HitResult :type Hit%20Result%20Structure%20%28by%20ref%29) (bIsOverlap :type Boolean))
  (bind _waterhit (Variables|Default|GetWaterHit))
  (bind (_bblockinghit _binitialoverlap _time _distance _location _impactpoint _normal _impactnormal _physmat) (Collision|BreakHitResult _waterhit))
  (bind _waterinteractioncomponent (Variables|Water|GetWaterInteractionComponent))
  (bind (_waterplanelocation _waterplanenormal _watersurfaceposition _waterdepth _waterbodyidx) (FX|GetLastWaterSurfaceInfo _waterinteractioncomponent :!class /Script/FortniteGame.FortWaterInteractionComponent))
  ;[0 0.298 1] Water Impact Projectile
  (Networking|CanExecuteCosmeticEvents
    (:True
      (if bIsOverlap
        (Variables|Default|SetWaterHit HitResult)
        (bind _output_get (Variables|Default|SetWaterHitMat _physmat))
        (Utilities|IsValid _output_get
          (:Is%20Valid
            (CallFunction|PlayImpactSFX false true)
            (if (Utilities|Enum|Equal(Enum) (Variables|/Script/PhysicsCore.PhysicalMaterial|GetSurfaceType _output_get) "SurfaceType8")
              (Utilities|FlowControl|Sequence
                (:then_0
                  _)
                (:then_1
                  (Niagara|SetNiagaraVariable(Bool) (Variables|Default|GetNS_CoreSniper_Bullet_Trail) "Deactive" true))))
            (if (and (Utilities|Enum|Equal(Enum) (Variables|/Script/PhysicsCore.PhysicalMaterial|GetSurfaceType _output_get) "SurfaceType8") (Pawn|IsLocallyControlled (Game|GetInstigator)))
              (bind _returnvalue (Effects|SpawnEmitteratLocation (Variables|Default|GetWaterImpactFX) _location (Math|Conversions|RotationFromXVector _waterplanenormal)))
              (Effects|Components|ParticleSystem|SetFloatParameter _returnvalue "WaterDepth" _waterdepth)
              (Effects|Components|ParticleSystem|SetFloatParameter _returnvalue "WaterBodyIndex" (Math|Conversions|ToFloat(Integer) _waterbodyidx)))))))))

(event EventOnSniperTraceWaterHit
  (params (WaterHit :type Hit%20Result%20Structure%20%28by%20ref%29) (WaterBody :type Fort%20Water%20Body%20Actor%20Object%20Reference))
  (bind _waterhit (Variables|Default|GetWaterHit))
  (bind (_bblockinghit _binitialoverlap _time _distance _location _impactpoint _normal _impactnormal _physmat) (Collision|BreakHitResult _waterhit))
  (bind _waterdepthtrace (Variables|Default|GetWaterDepthTrace))
  ;[0 0.298 1] Water Impact Trace
  (Variables|Default|SetWaterHit WaterHit)
  (Utilities|FlowControl|Sequence
    (:then_0
      (Utilities|FlowControl|DoOnce false
        (:Completed
          (Variables|Default|SetTraceImpactLocation _location)
          (Variables|Default|SetTraceImpactNormal _impactnormal))))
    (:then_1
      (Networking|CanExecuteCosmeticEvents
        (:True
          (bind _output_get (Variables|Default|SetWaterHitMat _physmat))
          (Utilities|IsValid _output_get
            (:Is%20Valid
              (if (Utilities|Enum|Equal(Enum) (Variables|/Script/PhysicsCore.PhysicalMaterial|GetSurfaceType _output_get) "SurfaceType8")
                (Niagara|SetNiagaraVariable(Bool) (Variables|Default|GetNS_CoreSniper_Bullet_Trail) "Deactive" true)
                (Utilities|FlowControl|Delay 0.01)
                (CallFunction|PlayImpactSFX false true)
                (bind _outwatersurfaceinfo (WaterBody|GetWaterSurfaceInfoStruct WaterBody (Variables|Default|GetTraceImpactLocation) true false false))
                (Variables|Default|SetWaterDepthTrace (.WaterDepth (Utilities|Struct|BreakWaterSurfaceInfo _outwatersurfaceinfo)))
                (Variables|Default|SetWaterIndexTrace (.WaterBodyIdx (Utilities|Struct|BreakWaterSurfaceInfo _outwatersurfaceinfo)))
                (bind _returnvalue (Effects|SpawnEmitteratLocation (Variables|Default|GetWaterImpactFX) (Variables|Default|GetTraceImpactLocation) (Math|Rotator|MakeRotfromX (Variables|Default|GetTraceImpactNormal))))
                (Development|PrintString (Utilities|String|ToString(Float) _waterdepthtrace))
                (Effects|Components|ParticleSystem|SetFloatParameter _returnvalue "WaterDepth" _waterdepthtrace)
                (Effects|Components|ParticleSystem|SetFloatParameter _returnvalue "WaterBodyIndex" (Math|Conversions|ToFloat(Integer) (Variables|Default|GetWaterIndexTrace)))))))))))

(fn PlayImpactSFX
  (params (IsPlayer :type Boolean) (IsWater :type Boolean))
  (bind _returnvalue (Transformation|GetActorLocation))
  (if IsWater
    (Audio|PlaySoundatLocation (Variables|Default|GetSniper_Rifle_WaterImpactSound) _returnvalue)
    (Variables|Default|SetbHasPlayedWaterSound true)
    (elif (Variables|Default|GetbHasPlayedWaterSound)
      _
      (else
        (Audio|PlaySoundatLocation (select IsPlayer (Variables|Default|GetSniper_Rifle_PlayerImpactSound) (Variables|Default|GetSniper_Rifle_SurfaceImpactSound)) _returnvalue)))))

(fn ConstructionScript
  (Projectile|SetGravityScale (ScalableFloat|GetValueatLevel (Variables|Default|GetProjectileGravityScale))))
