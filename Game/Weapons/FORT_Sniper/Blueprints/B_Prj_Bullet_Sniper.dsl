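; Sniper bullet projectile Blueprint that sets gravity from a scalable float and drives the bullet's cosmetic effects.
; BeginPlay seeds the Niagara trail with muzzle origin and timing, briefly toggles it on mobile as a hack, and disables capsule overlaps where cosmetics don't run.
; Water hits from overlap or sniper trace share gated, do-once logic that reads surface depth and index, plays impact sound and splash, and deactivates the trail, as does stopping.
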(event EventBeginPlay
  (bind _bullettrailparticlecomponent (Variables|FortProjectileSniperAthena|GetBulletTrailParticleComponent))
  (bind _bullettrailparticlecomponent_2 (Variables|FortProjectileSniperAthena|GetBulletTrailParticleComponent))
  (|CallParentFunction)
  (Networking|CanExecuteCosmeticEvents
    (:True
      (bind _asfort_player_pawn (Utilities|Casting|CastToFortPlayerPawn (Variables|Actor|GetInstigator)))
      ;[1 1 1] Hacky fix for FORT-587865
      (if (Fortnite|IsMobilePlatform)
        (Components|Activation|Deactivate _bullettrailparticlecomponent_2)
        (Utilities|FlowControl|Delay 0.04)
        (Components|Activation|Activate _bullettrailparticlecomponent_2 false))
      (Niagara|SetNiagaraVariable(Vector3) _bullettrailparticlecomponent "BulletOrigin" (Effects|GetMuzzleLocation (Variables|/Script/FortniteGame.FortPawn|GetCurrentWeapon _asfort_player_pawn) -1))
      (Niagara|SetNiagaraVariable(Float) _bullettrailparticlecomponent "StartTime" (Utilities|Time|GetGameTimeinSeconds))
      (Niagara|SetNiagaraVariable(Float) _bullettrailparticlecomponent "WPODuration" 0.2))
    (:False
      (Variables|/Script/Engine.PrimitiveComponent|SetbGenerateOverlapEvents false (Variables|Collision|GetCapsuleComponent)) ; Overlaps only needed on clients for some FX interaction
    )))

(event OnProjectileStop(ProjectileMovementComponent)
  (params (ImpactResult :type Hit%20Result%20Structure%20%28by%20ref%29))
  (Networking|CanExecuteCosmeticEvents
    (:True
      (bind _bullettrailparticlecomponent (Variables|FortProjectileSniperAthena|GetBulletTrailParticleComponent)
        (:then
          (Niagara|SetNiagaraVariable(Bool) _bullettrailparticlecomponent "Deactive" true))
        (:else)))))

(event EventOnTouched
  (params (OtherActor :type Actor%20Object%20Reference) (OtherComp :type Primitive%20Component%20Object%20Reference) (HitResult :type Hit%20Result%20Structure%20%28by%20ref%29) (bIsOverlap :type Boolean))
  (bind (_bblockinghit _binitialoverlap _time _distance _location _impactpoint _normal) (Collision|BreakHitResult (Variables|Default|GetWaterHit)))
  (bind (_bblockinghit_2 _binitialoverlap_2 _time_2 _distance_2 _location_2) (Collision|BreakHitResult (Variables|Default|GetWaterHit)))
  ;[0 0.298 1] Water Impact Projectile
  (Networking|CanExecuteCosmeticEvents
    (:True
      (Utilities|FlowControl|Gate false
        (:Exit
          (if bIsOverlap
            (Variables|Default|SetWaterHit HitResult)
            (bind _asfort_water_body_actor (Utilities|Casting|CastToFortWaterBodyActor OtherActor))
            (Utilities|FlowControl|Sequence
              (:then_0
                (bind _outwatersurfaceinfo (WaterBody|GetWaterSurfaceInfoStruct _asfort_water_body_actor _location_2 true false false))
                (Variables|Default|SetWaterDepth (.WaterDepth (Utilities|Struct|BreakWaterSurfaceInfo _outwatersurfaceinfo)))
                (Variables|Default|SetWaterIndex (.WaterBodyIdx (Utilities|Struct|BreakWaterSurfaceInfo _outwatersurfaceinfo)))
                (Variables|Default|SetTraceImpactNormal (Utilities|Struct|BreakWaterSurfaceInfo _outwatersurfaceinfo))
                (Utilities|FlowControl|DoOnce false
                  (:Completed
                    (CallFunction|PlayImpactSFX "SurfaceType8")
                    (bind _returnvalue (Niagara|SpawnSystematLocation (Variables|Default|GetNiagaraWaterImpactFX) _location (Math|Conversions|RotationFromXVector _normal)))
                    (Effects|Components|ParticleSystem|SetFloatParameter _returnvalue "WaterDepth" (Variables|Default|GetWaterDepth))
                    (Effects|Components|ParticleSystem|SetFloatParameter _returnvalue "WaterBodyIndex" (Math|Conversions|ToFloat(Integer) (Variables|Default|GetWaterIndex))))))
              (:then_1
                (Niagara|SetNiagaraVariable(Bool) (Variables|FortProjectileSniperAthena|GetBulletTrailParticleComponent) "Deactive" true)))))) ; Block this if the water trace hit
    )))

(event EventOnSniperTraceWaterHit
  (params (WaterHit :type Hit%20Result%20Structure%20%28by%20ref%29) (WaterBody :type Fort%20Water%20Body%20Actor%20Object%20Reference))
  (bind (_bblockinghit _binitialoverlap _time _distance _location _impactpoint _normal) (Collision|BreakHitResult (Variables|Default|GetWaterHit)))
  (bind (_bblockinghit_2 _binitialoverlap_2 _time_2 _distance_2 _location_2) (Collision|BreakHitResult (Variables|Default|GetWaterHit)))
  ;[0 0.298 1] Water Impact Trace
  (Utilities|FlowControl|DoOnce false
    (:Completed
      (Utilities|IsValid WaterBody
        (:Is%20Valid
          (bind _output_get (Variables|Default|SetWaterHit WaterHit))
          (Utilities|FlowControl|Sequence
            (:then_0
              ;[0 0.298 1] Water Impact Projectile
              (Utilities|FlowControl|Gate false
                (:Exit
                  (if (stale EventOnTouched bIsOverlap)
                    (Variables|Default|SetWaterHit (stale EventOnTouched HitResult))
                    (bind _asfort_water_body_actor (Utilities|Casting|CastToFortWaterBodyActor (stale EventOnTouched OtherActor)))
                    (Utilities|FlowControl|Sequence
                      (:then_0
                        (bind _outwatersurfaceinfo (WaterBody|GetWaterSurfaceInfoStruct _asfort_water_body_actor _location_2 true false false))
                        (Variables|Default|SetWaterDepth (.WaterDepth (Utilities|Struct|BreakWaterSurfaceInfo _outwatersurfaceinfo)))
                        (Variables|Default|SetWaterIndex (.WaterBodyIdx (Utilities|Struct|BreakWaterSurfaceInfo _outwatersurfaceinfo)))
                        (Variables|Default|SetTraceImpactNormal (Utilities|Struct|BreakWaterSurfaceInfo _outwatersurfaceinfo)))
                      (:then_1
                        (Niagara|SetNiagaraVariable(Bool) (Variables|FortProjectileSniperAthena|GetBulletTrailParticleComponent) "Deactive" true :!label _label)))))) ; Block this if the water trace hit
            )
            (:then_1
              (Utilities|FlowControl|Delay 0.01)
              (bind _outwatersurfaceinfo_2 (WaterBody|GetWaterSurfaceInfoStruct WaterBody (.Location (Collision|BreakHitResult _output_get)) true false false))
              (Variables|Default|SetWaterDepth (.WaterDepth (Utilities|Struct|BreakWaterSurfaceInfo _outwatersurfaceinfo_2)))
              (Variables|Default|SetWaterIndex (.WaterBodyIdx (Utilities|Struct|BreakWaterSurfaceInfo _outwatersurfaceinfo_2))))
            (:then_2
              (goto _label)))
          (Utilities|FlowControl|DoOnce false
            (:Completed
              (CallFunction|PlayImpactSFX "SurfaceType8")
              (bind _returnvalue (Niagara|SpawnSystematLocation (Variables|Default|GetNiagaraWaterImpactFX) _location (Math|Conversions|RotationFromXVector _normal)))
              (Effects|Components|ParticleSystem|SetFloatParameter _returnvalue "WaterDepth" (Variables|Default|GetWaterDepth))
              (Effects|Components|ParticleSystem|SetFloatParameter _returnvalue "WaterBodyIndex" (Math|Conversions|ToFloat(Integer) (Variables|Default|GetWaterIndex))))))))))

(fn ConstructionScript
  (Projectile|SetGravityScale (ScalableFloat|GetValueatLevel (Variables|Default|GetProjectileGravityScale))))
