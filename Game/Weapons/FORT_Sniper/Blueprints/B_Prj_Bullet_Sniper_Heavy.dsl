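; Heavy sniper bullet projectile Blueprint that adds cosmetic water impact effects and sound.
; Projectile overlaps and sniper trace water hits both hide the bullet trail, play water impact audio and, after a tiny delay, spawn Niagara splashes fed water depth and body index.
; The overlap splash requires a water body and locally controlled instigator, using the water interaction component's surface data.
; The trace path checks surface type, caches its first impact and queries the water body.
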(event EventOnTouched
  (params (OtherActor :type Actor%20Object%20Reference) (OtherComp :type Primitive%20Component%20Object%20Reference) (HitResult :type Hit%20Result%20Structure%20%28by%20ref%29) (bIsOverlap :type Boolean))
  (bind _waterhit (Variables|Default|GetWaterHit))
  (bind (_bblockinghit _binitialoverlap _time _distance _location _impactpoint _normal _impactnormal _physmat _hitactor) (Collision|BreakHitResult _waterhit))
  (bind _waterinteractioncomponent (Variables|Water|GetWaterInteractionComponent))
  (bind (_waterplanelocation _waterplanenormal _watersurfaceposition _waterdepth _waterbodyidx) (FX|GetLastWaterSurfaceInfo _waterinteractioncomponent :!class /Script/FortniteGame.FortWaterInteractionComponent))
  (bind (_aswater_body _bsuccess) (Utilities|Casting|CastToWaterBody _hitactor))
  ;[0 0.298 1] Water Impact Projectile
  (Networking|CanExecuteCosmeticEvents
    (:True
      (if bIsOverlap
        (Variables|Default|SetWaterHit HitResult)
        (bind _output_get (Variables|Default|SetWaterHitMat _physmat))
        (Utilities|IsValid _output_get
          (:Is%20Valid
            (Rendering|SetVisibility (Variables|FortProjectileSniperAthena|GetBulletTrailParticleComponent) false false)
            (CallFunction|PlayImpactSFX "SurfaceType8")
            (if (and _bsuccess (Pawn|IsLocallyControlled (Game|GetInstigator)))
              (Utilities|FlowControl|Delay 0.01)
              (bind _returnvalue (Niagara|SpawnSystematLocation (Variables|Default|GetNiagaraWaterImpactFX) _location (Math|Conversions|RotationFromXVector _waterplanenormal)))
              (Effects|Components|ParticleSystem|SetFloatParameter _returnvalue "WaterDepth" _waterdepth)
              (Effects|Components|ParticleSystem|SetFloatParameter _returnvalue "WaterBodyIndex" (Math|Conversions|ToFloat(Integer) _waterbodyidx)))))))))

(event EventOnSniperTraceWaterHit
  (params (WaterHit :type Hit%20Result%20Structure%20%28by%20ref%29) (WaterBody :type Fort%20Water%20Body%20Actor%20Object%20Reference))
  (bind _waterhit (Variables|Default|GetWaterHit))
  (bind (_bblockinghit _binitialoverlap _time _distance _location _impactpoint _normal _impactnormal _physmat) (Collision|BreakHitResult _waterhit))
  ;[0 0.298 1] Water Impact Trace
  (Variables|Default|SetWaterHit WaterHit)
  (Utilities|FlowControl|Sequence
    (:then_0
      (Utilities|FlowControl|DoOnce false
        (:Completed
          (Variables|Default|SetTraceImpactLocation _location)
          (Variables|Default|SetTraceImpactNormal _impactnormal))))
    (:then_1
      (Networking|CanExecuteCosmeticEvents
        (:True
          (bind _output_get (Variables|Default|SetWaterHitMat _physmat))
          (Utilities|IsValid _output_get
            (:Is%20Valid
              (if (Utilities|Enum|Equal(Enum) (Variables|/Script/PhysicsCore.PhysicalMaterial|GetSurfaceType _output_get) "SurfaceType8")
                (Rendering|SetVisibility (Variables|FortProjectileSniperAthena|GetBulletTrailParticleComponent) false false)
                (Utilities|FlowControl|Delay 0.01)
                (CallFunction|PlayImpactSFX "SurfaceType8")
                (bind _outwatersurfaceinfo (WaterBody|GetWaterSurfaceInfoStruct WaterBody (Variables|Default|GetTraceImpactLocation) true false false))
                (Variables|Default|SetWaterDepth (.WaterDepth (Utilities|Struct|BreakWaterSurfaceInfo _outwatersurfaceinfo)))
                (Variables|Default|SetWaterIndex (.WaterBodyIdx (Utilities|Struct|BreakWaterSurfaceInfo _outwatersurfaceinfo)))
                (bind _returnvalue (Niagara|SpawnSystematLocation (Variables|Default|GetNiagaraWaterImpactFX) (Variables|Default|GetTraceImpactLocation) (Math|Rotator|MakeRotfromX (Variables|Default|GetTraceImpactNormal))))
                (Effects|Components|ParticleSystem|SetFloatParameter _returnvalue "WaterDepth" (Variables|Default|GetWaterDepth))
                (Effects|Components|ParticleSystem|SetFloatParameter _returnvalue "WaterBodyIndex" (Math|Conversions|ToFloat(Integer) (Variables|Default|GetWaterIndex)))))))))))

(fn ConstructionScript
  (|CallParentFunction))
