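; Projectile Blueprint for the suppressed auto sniper bullet, handling gravity scale and cosmetic tracer, trail and impact effects.
; BeginPlay feeds muzzle origin and start time to the tracer material for world-position-offset animation.
; On stop, it plays player or surface audio, shuts off the trail and whip tracker, and spawns non-mobile bullet-hole decals.
; Water overlap (locally controlled) and trace events spawn depth-aware splashes; the trace path still prints depth.
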(event EventBeginPlay
  (bind _tracermesh (Variables|Default|GetTracerMesh))
  (|CallParentFunction)
  (Networking|CanExecuteCosmeticEvents
    (:True
      (bind _asfort_player_pawn (Utilities|Casting|CastToFortPlayerPawn (Variables|Actor|GetInstigator)))
      (Rendering|Material|SetVectorParameterValueonMaterials _tracermesh "BulletOrigin" (Effects|GetMuzzleLocation (Variables|/Script/FortniteGame.FortPawn|GetCurrentWeapon _asfort_player_pawn) -1))
      (Rendering|Material|SetScalarParameterValueonMaterials _tracermesh "StartTime" (Utilities|Time|GetGameTimeinSeconds))
      (Rendering|Material|SetScalarParameterValueonMaterials _tracermesh "WPODuration" 0.2))))

(event EventOnStop
  (params (Hit :type Hit%20Result%20Structure%20%28by%20ref%29))
  (bind (_bblockinghit _binitialoverlap _time _distance _location _impactpoint _normal _impactnormal _physmat _hitactor) (Collision|BreakHitResult Hit))
  (bind (_asfort_pawn _bsuccess) (Utilities|Casting|CastToFortPawn _hitactor))
  (bind _p_sniper_bullet_trail (Variables|Default|GetP_Sniper_Bullet_Trail))
  (Utilities|FlowControl|Sequence
    (:then_0
      (Variables|/Script/Engine.ProjectileMovementComponent|SetProjectileGravityScale 0.0 (Variables|Movement|GetProjectileMovementComponent))
      (Class|BulletWhipTrackerComponentBase|Disable (Variables|Visual|GetBulletWhipTrackerComponent)))
    (:then_1
      (Networking|CanExecuteCosmeticEvents
        (:True
          (Audio|PlaySoundatLocation (select _bsuccess (Variables|Default|GetSniper_Rifle_PlayerImpactSound) (Variables|Default|GetSniper_Rifle_SurfaceImpactSound)) (Transformation|GetActorLocation))
          (Transformation|DetachFromComponent _p_sniper_bullet_trail "KeepWorld" "KeepWorld" "KeepWorld" true)
          (Components|Activation|Deactivate _p_sniper_bullet_trail)
          (if (not (or (Fortnite|IsMobilePlatform) _bsuccess))
            (bind _returnvalue (Rendering|Decal|SpawnDecalatLocation "/Game/Weapons/Materials/MD_Decal_BulletHole_Lifetime_Inst.MD_Decal_BulletHole_Lifetime_Inst" (Variables|Default|GetDecalSize) _location (Math|Conversions|RotationFromXVector (* _impactnormal -1.0)) 4.5))
            (Rendering|Components|Decal|SetFadeOut _returnvalue 0.0 4.5 true)
            (Utilities|FlowControl|Delay 0.2))
          (Rendering|SetVisibility (Variables|Default|GetTracerMesh) false false))))))

(event EventOnTouched
  (params (OtherActor :type Actor%20Object%20Reference) (OtherComp :type Primitive%20Component%20Object%20Reference) (HitResult :type Hit%20Result%20Structure%20%28by%20ref%29) (bIsOverlap :type Boolean))
  (bind _waterhit (Variables|Default|GetWaterHit))
  (bind (_bblockinghit _binitialoverlap _time _distance _location _impactpoint _normal _impactnormal _physmat) (Collision|BreakHitResult _waterhit))
  (bind _waterinteractioncomponent (Variables|Water|GetWaterInteractionComponent))
  (bind (_waterplanelocation _waterplanenormal _watersurfaceposition _waterdepth _waterbodyidx) (FX|GetLastWaterSurfaceInfo _waterinteractioncomponent :!class /Script/FortniteGame.FortWaterInteractionComponent))
  ;[0 0.298 1] Water Impact Projectile
  (Networking|CanExecuteCosmeticEvents
    (:True
      (if bIsOverlap
        (Variables|Default|SetWaterHit HitResult)
        (bind _output_get (Variables|Default|SetWaterHitMat _physmat))
        (Utilities|IsValid _output_get
          (:Is%20Valid
            (Rendering|SetVisibility (Variables|Default|GetP_Sniper_Bullet_Trail) false false)
            (CallFunction|PlayImpactSFX false true)
            (if (and (Utilities|Enum|Equal(Enum) (Variables|/Script/PhysicsCore.PhysicalMaterial|GetSurfaceType _output_get) "SurfaceType8") (Pawn|IsLocallyControlled (Game|GetInstigator)))
              (bind _returnvalue (Effects|SpawnEmitteratLocation (Variables|Default|GetWaterImpactFX) _location (Math|Conversions|RotationFromXVector _waterplanenormal)))
              (Effects|Components|ParticleSystem|SetFloatParameter _returnvalue "WaterDepth" _waterdepth)
              (Effects|Components|ParticleSystem|SetFloatParameter _returnvalue "WaterBodyIndex" (Math|Conversions|ToFloat(Integer) _waterbodyidx)))))))))

(event EventOnSniperTraceWaterHit
  (params (WaterHit :type Hit%20Result%20Structure%20%28by%20ref%29) (WaterBody :type Fort%20Water%20Body%20Actor%20Object%20Reference))
  (bind _waterhit (Variables|Default|GetWaterHit))
  (bind (_bblockinghit _binitialoverlap _time _distance _location _impactpoint _normal _impactnormal _physmat) (Collision|BreakHitResult _waterhit))
  (bind _waterdepthtrace (Variables|Default|GetWaterDepthTrace))
  ;[0 0.298 1] Water Impact Trace
  (Variables|Default|SetWaterHit WaterHit)
  (Utilities|FlowControl|Sequence
    (:then_0
      (Utilities|FlowControl|DoOnce false
        (:Completed
          (Variables|Default|SetTraceImpactLocation _location)
          (Variables|Default|SetTraceImpactNormal _impactnormal))))
    (:then_1
      (Networking|CanExecuteCosmeticEvents
        (:True
          (bind _output_get (Variables|Default|SetWaterHitMat _physmat))
          (Utilities|IsValid _output_get
            (:Is%20Valid
              (if (Utilities|Enum|Equal(Enum) (Variables|/Script/PhysicsCore.PhysicalMaterial|GetSurfaceType _output_get) "SurfaceType8")
                (Rendering|SetVisibility (Variables|Default|GetP_Sniper_Bullet_Trail) false false)
                (CallFunction|PlayImpactSFX false true)
                (bind _outwatersurfaceinfo (WaterBody|GetWaterSurfaceInfoStruct WaterBody (Variables|Default|GetTraceImpactLocation) true false false))
                (Variables|Default|SetWaterDepthTrace (.WaterDepth (Utilities|Struct|BreakWaterSurfaceInfo _outwatersurfaceinfo)))
                (Variables|Default|SetWaterIndexTrace (.WaterBodyIdx (Utilities|Struct|BreakWaterSurfaceInfo _outwatersurfaceinfo)))
                (bind _returnvalue (Effects|SpawnEmitteratLocation (Variables|Default|GetWaterImpactFX) (Variables|Default|GetTraceImpactLocation) (Math|Rotator|MakeRotfromX (Variables|Default|GetTraceImpactNormal))))
                (Development|PrintString (Utilities|String|ToString(Float) _waterdepthtrace))
                (Effects|Components|ParticleSystem|SetFloatParameter _returnvalue "WaterDepth" _waterdepthtrace)
                (Effects|Components|ParticleSystem|SetFloatParameter _returnvalue "WaterBodyIndex" (Math|Conversions|ToFloat(Integer) (Variables|Default|GetWaterIndexTrace)))))))))))

(fn PlayImpactSFX
  (params (Is%20Player :type Boolean) (Is%20Water :type Boolean))
  (bind _returnvalue (Transformation|GetActorLocation))
  (if Is%20Water
    (Audio|PlaySoundatLocation (Variables|Default|GetSniper_Rifle_WaterImpactSound) _returnvalue)
    (Variables|Default|SetbHasPlayedWaterSound true)
    (elif (Variables|Default|GetbHasPlayedWaterSound)
      _
      (else
        (Audio|PlaySoundatLocation (select Is%20Player (Variables|Default|GetSniper_Rifle_PlayerImpactSound) (Variables|Default|GetSniper_Rifle_SurfaceImpactSound)) _returnvalue)))))

(fn ConstructionScript
  (Projectile|SetGravityScale (ScalableFloat|GetValueatLevel (Variables|Default|GetProjectileGravityScale))))
