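# Two-sided unlit translucent glow material for the VaccBow sniper weapon's particle effects, usable on mesh particles.
# Particle color feeds emissive through MF_FX_EmissiveAdjustments, while particle alpha scales a procedural UV gradient mask.
# Scalar parameters blend in a broader gradient, bypass the sphere mask, and enable UV1-driven linear erosion and depth fade, which mobile skips.
# Dynamic parameters drive intensity and erosion offset; Width is named but unused.
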Material /Game/Weapons/FORT_Sniper/Materials/M_Glow1_VaccBow
graph {
 properties {
  bUseTranslucencyVertexFog=false
  bUsedWithMeshParticles=true
  blendMode=BLEND_Translucent
  pixelDepthOffsetMode=PDOM_Legacy
  shadingModel=MSM_Unlit
  translucencyPass=MTP_BeforeDOF
  twoSided=true
  usedShadingModels=Unlit
 }
 nodes [
  n1 = ParticleColor()
  MaterialFunctionCall(Emissive=@n1.RGB) {materialFunction=&MF_FX_EmissiveAdjustments}
  n2 = Subtract(A=ComponentMask(None=TextureCoordinate()) {r=true}) {constB=0.5}
  n3 = OneMinus(None=Abs(None=Multiply(A=@n2) {constB=2}))
  n4 = Multiply(A=@n3, B=@n3)
  n5 = DynamicParameter() {paramNames [Intensity, Width, ErodeUVOffset, Param4]}
  n6 = Multiply(A=@n5.Intensity) {constB=0.5}
  n7 = TextureCoordinate()
  n8 = Sine(None=Divide(A=ComponentMask(None=@n7) {r=true}))
  n9 = Clamp(None=Power(Base=ComponentMask(None=@n7) {g=true}) {constExponent=3})
  n10 = Multiply(A=Clamp(None=Power(Base=@n8) {constExponent=5}), B=OneMinus(None=Subtract(A=@n9) {constB=0.7}))
  n11 = Constant() {r=0.5}
  n12 = SphereMask(A=TextureCoordinate(), B=@n11, Radius=@n11) {hardnessPercent=0}
  n13 = Multiply(
   A=LinearInterpolate(A=Multiply(A=@n4, B=@n6), B=Multiply(A=@n10, B=@n5.Intensity), Alpha=ScalarParameter() {parameterName=UseBroaderGradient, sliderMax=1})
   B=LinearInterpolate(A=Power(Base=@n12) {constExponent=1}, B=Constant() {r=1}, Alpha=ScalarParameter() {parameterName=DisableSphereMask, sliderMax=1})
  )
  n14 = ComponentMask(None=TextureCoordinate() {coordinateIndex=1, vTiling=0.5}) {g=true}
  n15 = LinearInterpolate(Alpha=Clamp(None=Add(A=@n14, B=@n5.ErodeUVOffset) {constB=-1}) {minDefault=-1}) {constA=-1}
  n16 = LinearInterpolate(A=@n13, B=Subtract(A=@n13, B=OneMinus(None=@n15)), Alpha=ScalarParameter() {parameterName=UseLinearErode, sliderMax=1})
  n17 = DepthFade(Opacity=@n16) {fadeDistanceDefault=25}
  Multiply(
   A=@n1.A
   B=LinearInterpolate(
    A=@n16
    B=ShadingPathSwitch(Default=@n17, Mobile=@n16) {desc=Migrated from FeatureLevelSwitch}
    Alpha=ScalarParameter() {parameterName=UseDepthFade, sliderMax=1}
   )
  )
 ]
}
