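# Lit material for the battery on the v3 FORT sniper, usable on skeletal meshes, with a cyan glow.
# A cyan color parameter feeds the shared MF_FX_EmissiveAdjustments function, alongside plain constants and a second cyan parameter.
# A sphere mask over UVs with V tiling zeroed uses MeterEdge offset, radius and hardness parameters to define a meter edge.
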Material /Game/Weapons/FORT_Sniper/Materials/M_Sniper_v3_Battery
graph {
 properties {bUsedWithMeshDeformer=true, bUsedWithSkeletalMesh=true, pixelDepthOffsetMode=PDOM_Legacy, usedShadingModels=Default Lit}
 nodes [
  Constant()
  Constant() {r=0.2}
  MaterialFunctionCall(Emissive=VectorParameter() {defaultValue=<rgba 0.197 0.7074 0.7292 0>}.RGB) {materialFunction=&MF_FX_EmissiveAdjustments}
  SphereMask(
   A=TextureCoordinate() {vTiling=0}
   B=ComponentMask(None=VectorParameter() {parameterName=MeterEdge_Offset, defaultValue=<rgba 0.5 0.5 0 0>}.RGB) {g=true, r=true}
   Radius=ScalarParameter() {defaultValue=0.1, parameterName=MeterEdge_Radius}
   Hardness=ScalarParameter() {parameterName=MeterEdge_Hardness}
  )
  VectorParameter() {defaultValue=<rgba 0.1203 0.8281 0.8034 0>}
 ]
}
