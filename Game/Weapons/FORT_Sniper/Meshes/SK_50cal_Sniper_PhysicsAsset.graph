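# Physics asset for the .50 cal sniper rifle skeletal mesh, giving the gun and its bipod simple collision bodies for physics.
# A long capsule on Handle stands in for the rifle, with a small capsule for the bipod hub and a box for each leg.
# Constraints chain both legs to the hub and the hub to Handle, locking both swing axes so each joint rotates only around twist, like a hinge.
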PhysicsAsset /Game/Weapons/FORT_Sniper/Meshes/SK_50cal_Sniper_PhysicsAsset
bodies {
 FiddyCal_dyn_bipod [{shapeType=Capsule, length=2.3646, radius=1.7343, rotation=<rot 90 0 0>}]
 FiddyCal_dyn_bipod_l [{shapeType=Box, extentX=4.5, extentY=1.5, extentZ=18.5, center=<0 0 -10.2867>}]
 FiddyCal_dyn_bipod_r [{shapeType=Box, extentX=4.5, extentY=1.5, extentZ=18.5, center=<0 0 -10.2867>}]
 Handle [{shapeType=Capsule, length=127.5, radius=4.097, center=<40.9181 0.0001 8.5051>, rotation=<rot 90 359.9998 0>}]
}
constraints [
 {
  bone1Name=FiddyCal_dyn_bipod
  bone2Name=Handle
  swing1LimitDegrees=0
  swing1Motion=Locked
  swing2LimitDegrees=0
  swing2Motion=Locked
  twistLimitDegrees=45
 }
 {
  bone1Name=FiddyCal_dyn_bipod_r
  bone2Name=FiddyCal_dyn_bipod
  swing1LimitDegrees=0
  swing1Motion=Locked
  swing2LimitDegrees=0
  swing2Motion=Locked
  twistLimitDegrees=45
 }
 {
  bone1Name=FiddyCal_dyn_bipod_l
  bone2Name=FiddyCal_dyn_bipod
  swing1LimitDegrees=0
  swing1Motion=Locked
  swing2LimitDegrees=0
  swing2Motion=Locked
  twistLimitDegrees=45
 }
]
