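# Physics asset for the bolt-action sniper rifle skeletal mesh, giving the weapon simple collision for physics simulation and queries.
# A single capsule body on the Handle bone, rotated to lie lengthwise along the rifle, approximates the weapon's volume.
# It defines no constraints, so the rifle behaves as one rigid piece with no separately simulated parts.
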PhysicsAsset /Game/Weapons/FORT_Sniper/Meshes/SK_Bolt_Action_Sniper_Rifle_TT_Physics
bodies {Handle [{shapeType=Capsule, length=88.5055, radius=16.1731, center=<36.5646 0.5422 7.8815>, rotation=<rot 0 89.8127 -89.0328>}]}
constraints []
