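# Physics asset for the bolt-action sniper rifle with scope skeletal mesh, SK_Bolt_Rifle_With_Scope, in the weapons content.
# It defines a single body on the Handle bone, shaped as one long box offset from the bone origin.
# It has no constraints, so the rifle is simulated and collides as one rigid piece with no articulated parts.
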PhysicsAsset /Game/Weapons/FORT_Sniper/Meshes/SK_Bolt_Rifle_With_Scope_Physics
bodies {Handle [{shapeType=Box, extentX=110.4094, extentY=24.3474, extentZ=29.6858, center=<23.227 0.7561 6.5986>}]}
constraints []
