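# Physics asset for the second Fortnite sniper rifle skeletal mesh, giving the weapon a simple collision and physics shape.
# A single body on the Handle bone uses one box, offset forward and upward from the bone to cover the rifle.
# It has no constraints, so the weapon behaves as one rigid piece with no articulated parts.
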PhysicsAsset /Game/Weapons/FORT_Sniper/Meshes/SK_Sniper_02_Physics
bodies {Handle [{shapeType=Box, extentX=96.3928, extentY=9.1951, extentZ=30.1839, center=<22.9947 0.5097 8.2984>}]}
constraints []
