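# Physics asset for the SK_Sniper_Rifle_v3 weapon mesh, giving the sniper rifle a simplified shape for physics.
# A large capsule on the Handle bone and a thin, shorter capsule on the MuzzleRecoil bone approximate the rifle.
# A single constraint joins MuzzleRecoil to Handle, with equal angular limit values on both swing axes and twist.
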PhysicsAsset /Game/Weapons/FORT_Sniper/Meshes/SK_Sniper_Rifle_v3_PhysicsAsset
bodies {
 Handle [{shapeType=Capsule, length=57.7118, radius=17.7974, center=<33.4995 4.6398 8.2682>, rotation=<rot 0 89.2911 -88.4667>}]
 MuzzleRecoil [{shapeType=Capsule, length=24.2625, radius=2.0398, center=<-0.0129 -13.1747 0.0576>, rotation=<rot 0 0 -90>}]
}
constraints [{bone1Name=MuzzleRecoil, bone2Name=Handle, swing1LimitDegrees=45, swing2LimitDegrees=45, twistLimitDegrees=45}]
