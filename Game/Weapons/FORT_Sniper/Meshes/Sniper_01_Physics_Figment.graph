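# Physics asset for a sniper weapon mesh, the Figment variant of Sniper_01.
# A single box body on the Handle bone approximates the weapon as one long, narrow shape, offset from the bone origin.
# It defines no constraints, so there are no jointed or separately simulated parts.
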PhysicsAsset /Game/Weapons/FORT_Sniper/Meshes/Sniper_01_Physics_Figment
bodies {Handle [{shapeType=Box, extentX=125.117, extentY=11.1399, extentZ=38.2671, center=<33.9364 0.2302 7.5267>}]}
constraints []
